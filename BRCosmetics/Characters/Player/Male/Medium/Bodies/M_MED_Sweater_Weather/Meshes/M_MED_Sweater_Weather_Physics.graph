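# Physics asset for the Sweater Weather male medium body cosmetic, defining bodies and joints for its dangling ornament, strap and handle bones.
# The ornament forms a three-joint chain and the strap a two-joint chain from their base bones, while the handle hangs off the right thigh twist bone.
# Tight swing and twist limits restrict sway.
# Pelvis, spine, thigh and front strap base bodies have no constraints, apart from that twist bone.
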PhysicsAsset /BRCosmetics/Characters/Player/Male/Medium/Bodies/M_MED_Sweater_Weather/Meshes/M_MED_Sweater_Weather_Physics
bodies {
 dyn_handel [{shapeType=Box, extentX=4.0151, extentY=3.4938, extentZ=2.5, center=<0.1964 0.0001 0>}]
 dyn_ornament_01 [{shapeType=Capsule, length=1.4374, radius=1, center=<1.5531 0 -0.0001>, rotation=<rot 90 0 0>}]
 dyn_ornament_02 [{shapeType=Capsule, length=3.9318, radius=5.3754, center=<6.9643 -0.8301 1.1967>, rotation=<rot 0 83.4398 -80.6235>}]
 dyn_ornament_base [{shapeType=Box, extentX=0.75, extentY=0.75, extentZ=0.5}]
 dyn_ornament_parent [{shapeType=Capsule, length=1.6327, radius=1.0848, center=<1.4979 0 0>, rotation=<rot 90 0 0>}]
 dyn_strap_01 [{shapeType=Box, extentX=4.1809, extentY=10.728, extentZ=0.9584, center=<1.7806 0 -0.1308>, rotation=<rot -5 0 0>}]
 dyn_strap_base [{shapeType=Box, extentX=0.75, extentY=0.75, extentZ=0.75}]
 dyn_strap_front_base [{shapeType=Box, extentX=1.1875, extentY=1.1875, extentZ=1.1875}]
 dyn_strap_parent [{shapeType=Box, extentX=10.01, extentY=9.51, extentZ=1.0725, center=<5.0645 0 -0.2046>, rotation=<rot -5 0 0>}]
 pelvis [{shapeType=Box, extentX=3, extentY=3, extentZ=3}]
 spine_01 [{shapeType=Capsule, length=15, radius=15, center=<5.9567 0 0>, rotation=<rot 90 0 0>}]
 thigh_l [{shapeType=Capsule, length=20.75, radius=9.25, center=<-14.701 1.4641 0>, rotation=<rot 82.8589 180 180>}]
 thigh_r [{shapeType=Capsule, length=20.625, radius=9.375, center=<8.7079 0.0905 -1.0116>, rotation=<rot 88.935 176.4244 -179.9998>}]
 thigh_twist_01_r [{shapeType=Box, extentX=7.5, extentY=7.5, extentZ=7.5}]
}
constraints [
 {bone1Name=dyn_ornament_parent, bone2Name=dyn_ornament_base, swing1LimitDegrees=15, swing2LimitDegrees=15, twistLimitDegrees=8}
 {bone1Name=dyn_ornament_01, bone2Name=dyn_ornament_parent, swing1LimitDegrees=15, swing2LimitDegrees=15, twistLimitDegrees=8}
 {bone1Name=dyn_ornament_02, bone2Name=dyn_ornament_01, swing1LimitDegrees=15, swing2LimitDegrees=15, twistLimitDegrees=8}
 {bone1Name=dyn_handel, bone2Name=thigh_twist_01_r, swing1LimitDegrees=5, swing2LimitDegrees=5, twistLimitDegrees=3}
 {bone1Name=dyn_strap_parent, bone2Name=dyn_strap_base, swing1LimitDegrees=20, swing2LimitDegrees=15, twistLimitDegrees=3}
 {bone1Name=dyn_strap_01, bone2Name=dyn_strap_parent, swing1LimitDegrees=10, swing2LimitDegrees=15, twistLimitDegrees=8}
]
