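# Anim Blueprint for the Sweater Weather male body cosmetic, layering secondary motion onto the pose copied from the attached parent mesh.
# Constraints drive hoof finger bones and anchor ring, ornament and strap bones, with AnimDynamics on rings and a final RigidBody simulation.
# Crouch and thigh-rotation alphas from the event graph blend corrective bone offsets and scaling before RigidBody.
# Both left and right thigh alphas derive from thigh_r yaw relative to pelvis.
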AnimBlueprint /BRCosmetics/Characters/Player/Male/Medium/Bodies/M_MED_Sweater_Weather/Meshes/M_MED_Sweater_Weather_AnimBP
target_skeleton=&/M_MED_Sweater_Weather_Skeleton
graphs [
 AnimGraph {nodes [
   n1 = "Constraint - const_hoof_ring_01_l (R)"(ComponentPose=Local To Component(LocalPose=Copy Pose From Mesh() {bUseAttachedParent=true}), Alpha=1, ConstraintWeights_0=1) {
    alphaBoolBlend {blendOption=Linear}
    boneToModify {boneName=const_hoof_ring_01_l}
    constraintSetup [{transformType=Rotation, targetBone {boneName=ring_01_l}}]
    constraintWeights [1]
   }
   n2 = "Constraint - const_hoof_ring_02_l (R)"(ComponentPose=@n1, Alpha=1, ConstraintWeights_0=1) {
    alphaBoolBlend {blendOption=Linear}
    boneToModify {boneName=const_hoof_ring_02_l}
    constraintSetup [{transformType=Rotation, targetBone {boneName=ring_02_l}}]
    constraintWeights [1]
   }
   n3 = "Constraint - const_hoof_index_01_l (R)"(ComponentPose=@n2, Alpha=1, ConstraintWeights_0=1) {
    alphaBoolBlend {blendOption=Linear}
    boneToModify {boneName=const_hoof_index_01_l}
    constraintSetup [{transformType=Rotation, targetBone {boneName=index_01_l}}]
    constraintWeights [1]
   }
   n4 = "Constraint - const_hoof_index_02_l (R)"(ComponentPose=@n3, Alpha=1, ConstraintWeights_0=1) {
    alphaBoolBlend {blendOption=Linear}
    boneToModify {boneName=const_hoof_index_02_l}
    constraintSetup [{transformType=Rotation, targetBone {boneName=index_02_l}}]
    constraintWeights [1]
   }
   n5 = "Constraint - const_hoof_index_01_r (R)"(ComponentPose=@n4, Alpha=1, ConstraintWeights_0=1) {
    alphaBoolBlend {blendOption=Linear}
    boneToModify {boneName=const_hoof_index_01_r}
    constraintSetup [{transformType=Rotation, targetBone {boneName=index_01_r}}]
    constraintWeights [1]
   }
   n6 = "Constraint - const_hoof_index_02_r (R)"(ComponentPose=@n5, Alpha=1, ConstraintWeights_0=1) {
    alphaBoolBlend {blendOption=Linear}
    boneToModify {boneName=const_hoof_index_02_r}
    constraintSetup [{transformType=Rotation, targetBone {boneName=index_02_r}}]
    constraintWeights [1]
   }
   n7 = "Constraint - const_hoof_ring_01_r (R)"(ComponentPose=@n6, Alpha=1, ConstraintWeights_0=1) {
    alphaBoolBlend {blendOption=Linear}
    boneToModify {boneName=const_hoof_ring_01_r}
    constraintSetup [{transformType=Rotation, targetBone {boneName=ring_01_r}}]
    constraintWeights [1]
   }
   n8 = "Constraint - const_hoof_ring_02_r (R)"(ComponentPose=@n7, Alpha=1, ConstraintWeights_0=1) {
    alphaBoolBlend {blendOption=Linear}
    boneToModify {boneName=const_hoof_ring_02_r}
    constraintSetup [{transformType=Rotation, targetBone {boneName=ring_02_r}}]
    constraintWeights [1]
   }
   n9 = "Constraint - dyn_ring_r (PPPPP)"!(
    ComponentPose=@n8
    Alpha=1
    ConstraintWeights_0=0.5784
    ConstraintWeights_1:float=0.0331
    ConstraintWeights_2:float=0.055
    ConstraintWeights_3:float=0.2119
    ConstraintWeights_4:float=0.1216
   ) -> (Pose) {
    alphaBoolBlend {blendOption=Linear}
    boneToModify {boneName=dyn_ring_r}
    constraintSetup [
     {targetBone {boneName=clavicle_r}}
     {targetBone {boneName=upperarm_twist_01_r}}
     {targetBone {boneName=spine_03}}
     {targetBone {boneName=spine_04}}
     {targetBone {boneName=spine_05}}
    ]
    constraintWeights [1,1,1,1,1]
   }
   n10 = "Constraint - dyn_ring_l (PPPPPPP)"!(
    ComponentPose=@n9
    Alpha=1
    ConstraintWeights_0=0.3244
    ConstraintWeights_1:float=0.0032
    ConstraintWeights_2:float=0.0865
    ConstraintWeights_3:float=0.3221
    ConstraintWeights_4:float=0.2286
    ConstraintWeights_5:float=0.0315
    ConstraintWeights_6:float=0.0037
   ) -> (Pose) {
    alphaBoolBlend {blendOption=Linear}
    boneToModify {boneName=dyn_ring_l}
    constraintSetup [
     {targetBone {boneName=clavicle_l}}
     {targetBone {boneName=spine_02}}
     {targetBone {boneName=spine_03}}
     {targetBone {boneName=spine_04}}
     {targetBone {boneName=spine_05}}
     {targetBone {boneName=upperarm_twist_01_l}}
     {targetBone {boneName=neck_02}}
    ]
    constraintWeights [1,1,1,1,1,1,1]
   }
   n11 = Anim Dynamics - Bone: dyn_ring_l(ComponentPose=@n10, Alpha=1, ExternalForce="0,0,0") {
    angularDampingOverride=0.99
    bGravityOverrideInSimSpace=true
    bOverrideAngularDamping=true
    linearDampingOverride=0.99
    alphaBoolBlend {blendOption=Linear}
    boundBone {boneName=dyn_ring_l}
    componentAppliedLinearAccClamp=<3000 3000 3000>
    componentLinearAccScale=<0.25 0.25 0.25>
    componentLinearVelScale=<0.25 0.25 0.25>
    physicsBodyDefinitions [{
      sphereCollisionRadius=0
      boundBone {boneName=dyn_ring_l}
      boxExtents=<4 6 2>
      constraintSetup {angularLimitsMax=<0 10 5>, angularLimitsMin=<0 -15 -5>, angularTarget=<0 0 0>}
      localJointOffset=<1 0 0>
     }]
    relativeSpaceBone {boneName=root}
   }
   n12 = Anim Dynamics - Bone: dyn_ring_r(ComponentPose=@n11, Alpha=1, ExternalForce="0,0,0") {
    angularDampingOverride=0.99
    bGravityOverrideInSimSpace=true
    bOverrideAngularDamping=true
    linearDampingOverride=0.99
    alphaBoolBlend {blendOption=Linear}
    boundBone {boneName=dyn_ring_r}
    componentAppliedLinearAccClamp=<3000 3000 3000>
    componentLinearAccScale=<0.25 0.25 0.25>
    componentLinearVelScale=<0.25 0.25 0.25>
    physicsBodyDefinitions [{
      sphereCollisionRadius=0
      boundBone {boneName=dyn_ring_r}
      boxExtents=<4 6 2>
      constraintSetup {angularLimitsMax=<0 10 5>, angularLimitsMin=<0 -15 -5>, angularTarget=<0 0 0>}
      localJointOffset=<1 0 0>
     }]
    relativeSpaceBone {boneName=root}
   }
   n13 = "Constraint - dyn_ornament_base (PPPPPP)"!(
    ComponentPose=@n12
    Alpha=1
    ConstraintWeights_0=0.3997
    ConstraintWeights_1:float=0.152
    ConstraintWeights_2:float=0.0684
    ConstraintWeights_3:float=0.0017
    ConstraintWeights_4:float=0.376
    ConstraintWeights_5:float=0.0021
   ) -> (Pose) {
    alphaBoolBlend {blendOption=Linear}
    boneToModify {boneName=dyn_ornament_base}
    constraintSetup [
     {targetBone {boneName=spine_01}}
     {targetBone {boneName=spine_02}}
     {targetBone {boneName=spine_03}}
     {targetBone {boneName=spine_04}}
     {targetBone {boneName=pelvis}}
     {targetBone {boneName=thigh_twist_01_l}}
    ]
    constraintWeights [1,1,1,1,1,1]
   }
   n14 = "Transform (Modify) Bone - Bone: dyn_ornament_base"(
    ComponentPose=@n13
    Alpha=Get Crouch_Reg_Alpha() -> (Crouch_Reg_Alpha:double)
    Translation="-3.535,10,2.118"
    Rotation="0,0,0"
    Scale="1,1,1"
   ) {
    rotationMode=BMM_Additive
    rotationSpace=BCS_BoneSpace
    translationMode=BMM_Additive
    translationSpace=BCS_BoneSpace
    alphaBoolBlend {blendOption=Linear}
    alphaScaleBiasClamp {bInterpResult=true, interpSpeedDecreasing=3, interpSpeedIncreasing=4}
    boneToModify {boneName=dyn_ornament_base}
    translation=<-3.535 10 2.118>
   }
   n15 = "Constraint - dyn_strap_front_base (PPPPP)"!(
    ComponentPose=@n14
    Alpha=1
    ConstraintWeights_0=0.6168
    ConstraintWeights_1:float=0.0291
    ConstraintWeights_2:float=0.0122
    ConstraintWeights_3:float=0.0164
    ConstraintWeights_4:float=0.3255
   ) -> (Pose) {
    alphaBoolBlend {blendOption=Linear}
    boneToModify {boneName=dyn_strap_front_base}
    constraintSetup [
     {targetBone {boneName=pelvis}}
     {targetBone {boneName=spine_02}}
     {targetBone {boneName=spine_03}}
     {targetBone {boneName=thigh_twist_01_r}}
     {targetBone {boneName=spine_01}}
    ]
    constraintWeights [1,1,1,1,1]
   }
   n16 = "Transform (Modify) Bone - Bone: dyn_strap_front_base"(ComponentPose=@n15, Alpha=Get Thigh_Rot_R() -> (Thigh_Rot_R:double), Translation="-0.5,9.5,5", Rotation="30,0,25", Scale="1,1,1") {
    rotationMode=BMM_Additive
    rotationSpace=BCS_BoneSpace
    translationMode=BMM_Additive
    translationSpace=BCS_BoneSpace
    alphaBoolBlend {blendOption=Linear}
    alphaScaleBiasClamp {bInterpResult=true}
    boneToModify {boneName=dyn_strap_front_base}
    rotation {pitch=30, roll=25}
    translation=<-0.5 9.5 5>
   }
   n17 = "Transform (Modify) Bone - Bone: sub_thigh_l"(ComponentPose=@n16, Alpha=Get Thigh_Rot_L() -> (Thigh_Rot_L:double), Translation="0,0,0", Rotation="0,0,0", Scale="2.25,2.25,2.25") {
    scaleMode=BMM_Additive
    scaleSpace=BCS_BoneSpace
    alphaBoolBlend {blendOption=Linear}
    boneToModify {boneName=sub_thigh_l}
    scale=<2.25 2.25 2.25>
   }
   n18 = "Transform (Modify) Bone - Bone: sub_thigh_r"(ComponentPose=@n17, Alpha=Get Thigh_Rot_R(), Translation="0,0,0", Rotation="0,0,0", Scale="1.5,1.5,1.5") {
    scaleMode=BMM_Additive
    scaleSpace=BCS_BoneSpace
    alphaBoolBlend {blendOption=Linear}
    boneToModify {boneName=sub_thigh_r}
    scale=<1.5 1.5 1.5>
   }
   n19 = RigidBody(ComponentPose=@n18) {
    bClampLinearTranslationLimitToRefPose=true
    bOverrideWorldGravity=true
    overridePhysicsAsset=&/M_MED_Sweater_Weather_Physics
    simulationSpace=BaseBoneSpace
    alphaBoolBlend {blendOption=Linear}
    baseBoneRef {boneName=root}
    componentAppliedLinearAccClamp=<6000 6000 6000>
    componentLinearAccScale=<0.25 0.25 0.25>
    componentLinearVelScale=<0.25 0.25 0.25>
    overrideWorldGravity {z=-980}
   }
   Output Pose(Result=Component To Local(ComponentPose=@n19)) {layerGroup=None}
  ]}
 EventGraph {nodes [
   Event BlueprintUpdateAnimation()
   > n1 = Sequence()
   > Set Crouch_Reg_Alpha(Crouch_Reg_Alpha:double=SelectFloat(A=1, B=0, bPickA=Get bIsCrouching())) -> (Output_Get:double)
   n2 = GetOwnerSkeletalMeshComponent()
   n3 = BreakRotator(InRot=BreakTransform(InTransform=GetDeltaTransformFromRefPose(self=@n2, BoneName=thigh_r, BaseName=pelvis)).Rotation)
   Set thigh rotation_r(execute=@n1.then_1, thigh rotation_r:double=MapRangeClamped(Value=@n3.Yaw, InRangeA=-50, InRangeB=-90)) -> (Output_Get:double)
   > Set thigh_rot_r(thigh_rot_r=MapRangeClamped(Value=@n3.Yaw, InRangeA=0, InRangeB=-90))
   n4 = BreakTransform(InTransform=GetDeltaTransformFromRefPose(self=@n2, BoneName=thigh_r, BaseName=pelvis))
   > Set Thigh_Rot_L(Thigh_Rot_L:double=MapRangeClamped(Value=BreakRotator(InRot=@n4.Rotation).Yaw, InRangeA=-20, InRangeB=-100)) -> (Output_Get:double)
   TryGetPawnOwner()
  ]}
]
variables [
 Crouch_Reg_Alpha {type=number}
 thigh rotation_r {type=number}
 Thigh_Rot_L {type=number}
 Thigh_Rot_R {type=number}
]
