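# Physics asset for the face accessory of the male medium Sweater Weather cosmetic, giving limited physical motion to its dynamic hair and sock bones.
# Every bone has a single box body.
# Nine hair bodies each hang directly off dyn_hair_master with tight swing and very small twist limits.
# Sock bones 01 and 02 attach to dyn_sock_parent; 02 continues as a chain through 03 to 04, whose last joint locks twist and second swing axis.
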PhysicsAsset /BRCosmetics/Characters/Player/Male/Medium/Bodies/M_MED_Sweater_Weather/Meshes/Parts/M_MED_Sweater_Weather_FaceAcc_Physics
bodies {
 dyn_hair_01 [{shapeType=Box, extentX=2.5255, extentY=1.1691, extentZ=2.1054, center=<1.4386 0 0.0001>}]
 dyn_hair_02 [{shapeType=Box, extentX=2.7056, extentY=1.6356, extentZ=1.6886, center=<1.3615 0 0>}]
 dyn_hair_03 [{shapeType=Box, extentX=1.7832, extentY=1.7214, extentZ=1.423, center=<0.7173 0 0>}]
 dyn_hair_04 [{shapeType=Box, extentX=2.8879, extentY=2.5884, extentZ=2.1515, center=<1.5746 0 0>}]
 dyn_hair_05 [{shapeType=Box, extentX=3.4882, extentY=2.2844, extentZ=1.9165, center=<1.665 0 0>}]
 dyn_hair_06 [{shapeType=Box, extentX=3.1112, extentY=1.4908, extentZ=1.7494, center=<1.4592 0 0>}]
 dyn_hair_07 [{shapeType=Box, extentX=2.4911, extentY=1.1378, extentZ=1.5331, center=<1.0003 0 0>}]
 dyn_hair_08 [{shapeType=Box, extentX=1.76, extentY=1.01, extentZ=1.01, center=<0.57 0 0>}]
 dyn_hair_09 [{shapeType=Box, extentX=2.01, extentY=1.01, extentZ=1.1975, center=<0.5582 0 0>}]
 dyn_hair_master [{shapeType=Box, extentX=1.25, extentY=1.25, extentZ=1.25}]
 dyn_sock_01 [{shapeType=Box, extentX=4.1157, extentY=0.75, extentZ=3.7391, center=<1.7888 0.0001 0>}]
 dyn_sock_02 [{shapeType=Box, extentX=6.9946, extentY=5.4375, extentZ=1.2592, center=<3.367 -0.0001 0>}]
 dyn_sock_03 [{shapeType=Box, extentX=4.9934, extentY=5.5, extentZ=1.6234, center=<2.2762 0 0.0001>}]
 dyn_sock_04 [{shapeType=Box, extentX=4.7613, extentY=1.2784, extentZ=5.6915, center=<1.8924 0 0>}]
 dyn_sock_parent [{shapeType=Box, extentX=1.5, extentY=1.5, extentZ=1.5}]
}
constraints [
 {bone1Name=dyn_sock_01, bone2Name=dyn_sock_parent, swing1LimitDegrees=20, swing2LimitDegrees=20, twistLimitDegrees=3}
 {bone1Name=dyn_sock_02, bone2Name=dyn_sock_parent, swing1LimitDegrees=20, swing2LimitDegrees=40, twistLimitDegrees=4}
 {bone1Name=dyn_sock_03, bone2Name=dyn_sock_02, swing1LimitDegrees=10, swing2LimitDegrees=40, twistLimitDegrees=3}
 {
  bone1Name=dyn_sock_04
  bone2Name=dyn_sock_03
  swing1LimitDegrees=10
  swing2LimitDegrees=10
  swing2Motion=Locked
  twistLimitDegrees=6
  twistMotion=Locked
 }
 {bone1Name=dyn_hair_01, bone2Name=dyn_hair_master, swing1LimitDegrees=14, swing2LimitDegrees=12, twistLimitDegrees=1}
 {bone1Name=dyn_hair_02, bone2Name=dyn_hair_master, swing1LimitDegrees=10, swing2LimitDegrees=10, twistLimitDegrees=1}
 {bone1Name=dyn_hair_03, bone2Name=dyn_hair_master, swing1LimitDegrees=10, swing2LimitDegrees=10, twistLimitDegrees=1}
 {bone1Name=dyn_hair_04, bone2Name=dyn_hair_master, swing1LimitDegrees=10, swing2LimitDegrees=10, twistLimitDegrees=1}
 {bone1Name=dyn_hair_05, bone2Name=dyn_hair_master, swing1LimitDegrees=10, swing2LimitDegrees=10, twistLimitDegrees=1}
 {bone1Name=dyn_hair_06, bone2Name=dyn_hair_master, swing1LimitDegrees=10, swing2LimitDegrees=10, twistLimitDegrees=1}
 {bone1Name=dyn_hair_07, bone2Name=dyn_hair_master, swing1LimitDegrees=10, swing2LimitDegrees=10, twistLimitDegrees=1}
 {bone1Name=dyn_hair_08, bone2Name=dyn_hair_master, swing1LimitDegrees=10, swing2LimitDegrees=10, twistLimitDegrees=1}
 {bone1Name=dyn_hair_09, bone2Name=dyn_hair_master, swing1LimitDegrees=10, swing2LimitDegrees=10, twistLimitDegrees=1}
]
