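# Physics asset for the string lights face accessory on the male medium Sweater Weather cosmetic body, setting up the dangling light strands for physical motion.
# Four capsule chains (top and bottom, left and right) start at bodies constrained to neck_01, with tight swing and twist limits along each chain.
# The two bottom chains join at their tips to form a loop.
# Head, neck and box-shaped ear bodies supply collision volumes; head and ears have no constraints.
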PhysicsAsset /BRCosmetics/Characters/Player/Male/Medium/Bodies/M_MED_Sweater_Weather/Meshes/Parts/M_MED_Sweater_Weather_FaceAcc_StringLights_Physics
bodies {
 dyn_ear_l_base [{shapeType=Box, extentX=14.1875, extentY=5.4375, extentZ=8.25, center=<7.3186 -1.2903 -0.8929>, rotation=<rot -16.6281 9.8767 22.9096>}]
 dyn_ear_r_base [{shapeType=Box, extentX=14.1875, extentY=5.4375, extentZ=8.25, center=<7.1939 0.7861 -0.4655>, rotation=<rot -18.0989 -20.2273 -6.2584>}]
 dyn_stringlights_bottom_l_01 [{shapeType=Capsule, length=3.0884, radius=1.25, center=<2.0619 0 0>, rotation=<rot 90 0 0>}]
 dyn_stringlights_bottom_l_02 [{shapeType=Capsule, length=4.7134, radius=1.25, center=<3.0955 0 0>, rotation=<rot 90 0 0>}]
 dyn_stringlights_bottom_l_03 [{shapeType=Capsule, length=3.4634, radius=1.25, center=<1.69 0 0>, rotation=<rot 90 0 0>}]
 dyn_stringlights_bottom_l_base [{shapeType=Box, extentX=0.75, extentY=0.75, extentZ=0.75}]
 dyn_stringlights_bottom_l_parent [{shapeType=Capsule, length=4.2134, radius=1, center=<3.845 0 0>, rotation=<rot 90 0 0>}]
 dyn_stringlights_bottom_r_01 [{shapeType=Capsule, length=4.2134, radius=1.375, center=<2.6307 0 0>, rotation=<rot 90 0 0>}]
 dyn_stringlights_bottom_r_02 [{shapeType=Capsule, length=4.2134, radius=1.375, center=<2.6307 0 0>, rotation=<rot 90 0 0>}]
 dyn_stringlights_bottom_r_03 [{shapeType=Capsule, length=4.2134, radius=1.375, center=<2.6307 0 0>, rotation=<rot 90 0 0>}]
 dyn_stringlights_bottom_r_base [{shapeType=Box, extentX=0.75, extentY=0.75, extentZ=0.75}]
 dyn_stringlights_bottom_r_parent [{shapeType=Capsule, length=4.2134, radius=1.375, center=<2.6307 0 0>, rotation=<rot 90 0 0>}]
 dyn_stringlights_top_l_01 [{shapeType=Capsule, length=6.3697, radius=0.8438, center=<3.3853 1.9788 0.413>, rotation=<rot 90 0 -25.3123>}]
 dyn_stringlights_top_l_02 [{shapeType=Capsule, length=3.0259, radius=0.6875, center=<1.69 0 0>, rotation=<rot 90 0 0>}]
 dyn_stringlights_top_l_parent [{shapeType=Box, extentX=0.5, extentY=0.5, extentZ=0.5}]
 dyn_stringlights_top_r_01 [{shapeType=Capsule, length=5.4634, radius=1, center=<3.3725 0 0>, rotation=<rot 90 0 0>}]
 dyn_stringlights_top_r_02 [{shapeType=Capsule, length=2.8697, radius=0.8438, center=<1.69 0.1912 0>, rotation=<rot 90 0 0>}]
 dyn_stringlights_top_r_base [{shapeType=Box, extentX=0.5, extentY=0.5, extentZ=0.5}]
 dyn_stringlights_top_r_parent [{shapeType=Capsule, length=2.7134, radius=1, center=<1.69 0 0>, rotation=<rot 90 0 0>}]
 head [
  {shapeType=Capsule, length=18.5, radius=7.5, center=<2.0136 1.3876 -0.0001>, rotation=<rot 90 90 73.082>}
  {shapeType=Capsule, length=5.9375, radius=6.4688, center=<10.4309 1.1946 0>, rotation=<rot 90 0 -22.9703>}
 ]
 neck_01 [{shapeType=Capsule, length=12.25, radius=7.9425, center=<4.9374 0.7469 -0.0001>, rotation=<rot 90 14.0363 16.8488>}]
}
constraints [
 {
  bone1Name=dyn_stringlights_bottom_l_parent
  bone2Name=dyn_stringlights_bottom_l_base
  swing1LimitDegrees=30
  swing2LimitDegrees=30
  twistLimitDegrees=3
 }
 {
  bone1Name=dyn_stringlights_bottom_l_01
  bone2Name=dyn_stringlights_bottom_l_parent
  swing1LimitDegrees=10
  swing2LimitDegrees=10
  twistLimitDegrees=3
 }
 {
  bone1Name=dyn_stringlights_bottom_l_02
  bone2Name=dyn_stringlights_bottom_l_01
  swing1LimitDegrees=10
  swing2LimitDegrees=10
  twistLimitDegrees=3
 }
 {
  bone1Name=dyn_stringlights_bottom_l_03
  bone2Name=dyn_stringlights_bottom_l_02
  swing1LimitDegrees=20
  swing2LimitDegrees=10
  twistLimitDegrees=3
 }
 {
  bone1Name=dyn_stringlights_bottom_r_parent
  bone2Name=dyn_stringlights_bottom_r_base
  swing1LimitDegrees=10
  swing2LimitDegrees=30
  twistLimitDegrees=3
 }
 {
  bone1Name=dyn_stringlights_bottom_r_01
  bone2Name=dyn_stringlights_bottom_r_parent
  swing1LimitDegrees=20
  swing2LimitDegrees=10
  twistLimitDegrees=3
 }
 {
  bone1Name=dyn_stringlights_bottom_r_02
  bone2Name=dyn_stringlights_bottom_r_01
  swing1LimitDegrees=20
  swing2LimitDegrees=10
  twistLimitDegrees=3
 }
 {
  bone1Name=dyn_stringlights_bottom_r_03
  bone2Name=dyn_stringlights_bottom_r_02
  swing1LimitDegrees=20
  swing2LimitDegrees=10
  twistLimitDegrees=3
 }
 {
  bone1Name=dyn_stringlights_top_l_01
  bone2Name=dyn_stringlights_top_l_parent
  swing1LimitDegrees=20
  swing2LimitDegrees=10
  twistLimitDegrees=3
 }
 {bone1Name=dyn_stringlights_top_l_02, bone2Name=dyn_stringlights_top_l_01, swing1LimitDegrees=20, swing2LimitDegrees=10, twistLimitDegrees=3}
 {
  bone1Name=dyn_stringlights_top_r_parent
  bone2Name=dyn_stringlights_top_r_base
  swing1LimitDegrees=10
  swing2LimitDegrees=10
  twistLimitDegrees=3
 }
 {
  bone1Name=dyn_stringlights_top_r_01
  bone2Name=dyn_stringlights_top_r_parent
  swing1LimitDegrees=30
  swing2LimitDegrees=10
  twistLimitDegrees=3
 }
 {bone1Name=dyn_stringlights_top_r_02, bone2Name=dyn_stringlights_top_r_01, swing1LimitDegrees=20, swing2LimitDegrees=10, twistLimitDegrees=3}
 {bone1Name=dyn_stringlights_bottom_l_base, bone2Name=neck_01, swing1LimitDegrees=45, swing2LimitDegrees=45, twistLimitDegrees=45}
 {bone1Name=dyn_stringlights_bottom_r_base, bone2Name=neck_01, swing1LimitDegrees=45, swing2LimitDegrees=45, twistLimitDegrees=45}
 {bone1Name=dyn_stringlights_top_l_parent, bone2Name=neck_01, swing1LimitDegrees=45, swing2LimitDegrees=45, twistLimitDegrees=45}
 {bone1Name=dyn_stringlights_top_r_base, bone2Name=neck_01, swing1LimitDegrees=45, swing2LimitDegrees=45, twistLimitDegrees=45}
 {
  bone1Name=dyn_stringlights_bottom_l_03
  bone2Name=dyn_stringlights_bottom_r_03
  swing1LimitDegrees=45
  swing2LimitDegrees=45
  twistLimitDegrees=45
 }
]
